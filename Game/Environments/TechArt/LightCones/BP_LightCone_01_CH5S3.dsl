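; Configurable light cone actor with an optional animated fixture and optional real spot, surface and scatter lights.
; The construction script sizes the cone from angle and length, then feeds color, falloff, profile, WPO animation and time-of-day brightness into dynamic materials.
; The HLOD cone receives matching values through default custom primitive data.
; A location-seeded random value varies each instance and is printed on every construction; event graph events are empty.
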(event EventBeginPlay)

(event Collision|EventActorBeginOverlap
  (params (OtherActor :type Actor%20Object%20Reference)))

(event EventTick
  (params (DeltaSeconds :type Float%20%28single-precision%29)))

(fn BuildExposedAssetProviderVersion(Implementation)
  (params (Version :type Fort%20Exposed%20Asset%20Provider%20Version%20Structure%20%28by%20ref%29))
  (|CallParentFunction Version)
  (ExposedAssetProvider|AddVersion Version 1)
  (return))

(fn CollectAssetReferencestoAllow(Implementation)
  (params (Collector :type Fort%20Exposed%20Asset%20Reference%20Collector%20Structure%20%28by%20ref%29))
  (|CallParentFunction Collector)
  (ExposedAssetProvider|CollectObjectPropertyAssetReferences Collector 0 (Variables|Default|GetSpotlightBase))
  (ExposedAssetProvider|CollectObjectPropertyAssetReferences Collector 0 (Variables|Default|GetMaterial))
  (ExposedAssetProvider|CollectObjectPropertyAssetReferences Collector 0 (Variables|Default|GetHLODMaterial))
  (ExposedAssetProvider|CollectObjectPropertyAssetReferences Collector 0 (Variables|Default|GetMaterial%20LightFixture))
  (return))

(fn GetDynamicDestructionStaticMeshComponents
  (return (Utilities|Array|MakeArray (Variables|Debug|GetStaticMeshComponent) (Variables|Default|GetAnimatedLightFixture)) (Utilities|Array|MakeArray (Variables|Default|GetCone)) 0))

(fn ConstructionScript
  (bind (_r _g _b _a) (Math|Color|BreakColor (Variables|Default|GetLight%20Color%28rgb%29%20and%20Brightness%28a%29)))
  (bind (_x _y) (Math|Vector2D|BreakVector2D (Variables|ConeShape|GetXY%20Scale)))
  (bind _light_color_rgb__and_brightness_a (Variables|Default|GetLight%20Color%28rgb%29%20and%20Brightness%28a%29))
  (bind (_r_2 _g_2 _b_2 _a_2) (Math|Color|BreakColor _light_color_rgb__and_brightness_a))
  (bind _light_color_rgb__and_brightness_a_2 (Variables|Default|GetLight%20Color%28rgb%29%20and%20Brightness%28a%29))
  (bind (_r_3 _g_3 _b_3 _a_3) (Math|Color|BreakColor _light_color_rgb__and_brightness_a_2))
  (bind _light_shift (Variables|Falloff|GetLight%20Shift))
  (bind _returnvalue (Math|Float|Max(Float) _light_shift 1))
  (bind _light_multiplier (Variables|ActualLight|GetLight%20Multiplier))
  (bind _returnvalue_2 (* _a_2 _light_multiplier))
  (bind _returnvalue_3 (* _a_3 _returnvalue))
  (bind _wpo_animation_angle (Variables|Animation%28NeedsToBeEnabledInTheMaterial%29|GetWPO_Animation_Angle))
  (bind _returnvalue_4 (/ _wpo_animation_angle 2.0))
  (bind _returnvalue_5 (/ _returnvalue_4 3.6e+02))
  (bind _wpo_animation_speed (Variables|Animation%28NeedstobeEnabledintheMaterial%29|GetWPO_Animation_Speed))
  (bind _returnvalue_6 (/ _wpo_animation_speed 1e+02))
  (bind _light_shift_2 (Variables|Falloff|GetLight%20Shift))
  (bind _start_radius (Variables|ConeShape|GetStart%20Radius))
  (bind _cone_angle (Variables|ConeShape|GetCone%20Angle))
  (bind _cone (Variables|Default|GetCone))
  (bind _falloff (Variables|Falloff|GetFalloff))
  (bind _profile_side_contribution (Variables|Profile|GetProfile_Side_Contribution))
  (bind _brightness_morning (Variables|TODM|GetBrightness%20Morning))
  (bind _brightness_day (Variables|TODM|GetBrightness%20Day))
  (bind _brightness_evening (Variables|TODM|GetBrightness%20Evening))
  (bind _brightness_night (Variables|TODM|GetBrightness%20Night))
  (bind _profile_edgefalloff (Variables|Profile|GetProfile_EdgeFalloff))
  (bind _lightfixture_angle (Variables|Default|GetLightFixture%20Angle))
  (bind _mid_mat (Variables|Default|GetMID_Mat))
  (bind _profile_ground_brightness (Variables|Profile|GetProfile_Ground_Brightness))
  (bind _spotlight (Variables|Default|GetSpotLight))
  (bind _mid_mat_2 (Variables|Default|GetMID_Mat))
  (bind _stream (Variables|Default|GetStream))
  (Math|Random|SetRandomStreamSeed _stream (Math|Float|Truncate (Utilities|Operators|Add (.x (Transformation|GetActorLocation)) (.y (Transformation|GetActorLocation)) (.z (Transformation|GetActorLocation)))))
  (bind _output_get (Variables|Default|SetRandomValue (Math|Random|RandomFloatinRangefromStream _stream 0.0 1e+02)))
  (Development|PrintString (Utilities|String|ToString(Float) _output_get))
  (Utilities|FlowControl|Sequence
    (:then_0
      (if (Variables|Default|GetHasLightFixture)
        (Variables|Mesh|SetStaticMesh (Variables|Default|GetSpotlightBase))
        (Transformation|SetRelativeLocation _cone "0.000000,0.000000,60.000000" false false)
        (Transformation|SetRelativeRotation _cone (Math|Rotator|MakeRotator 0.0 (- _lightfixture_angle 9e+01)) false false)
        (Transformation|SetRelativeRotation (Variables|Default|GetAnimatedLightFixture) (Math|Rotator|MakeRotator 0.0 _lightfixture_angle) false false)
        (bind _material_lightfixture (Variables|Default|GetMaterial%20LightFixture)
          (:then
            (bind _returnvalue_7 (Rendering|Material|CreateDynamicMaterialInstance (Variables|Default|GetAnimatedLightFixture) 0 _material_lightfixture "None" :!class /Script/Engine.PrimitiveComponent))
            (Rendering|Material|SetScalarParameterValue _returnvalue_7 "WPOAnim_Speed" (/ (Variables|Animation%28NeedstobeEnabledintheMaterial%29|GetWPO_Animation_Speed) 1e+02) :!class /Script/Engine.MaterialInstanceDynamic)
            (Rendering|Material|SetScalarParameterValue _returnvalue_7 "WPOAnim_MaxAngle" (/ (/ (Variables|Animation%28NeedsToBeEnabledInTheMaterial%29|GetWPO_Animation_Angle) 2.0) 3.6e+02) :!class /Script/Engine.MaterialInstanceDynamic)
            (Rendering|Material|SetScalarParameterValue _returnvalue_7 "RandomValue" (Variables|Default|GetRandomValue) :!class /Script/Engine.MaterialInstanceDynamic)
            (Rendering|Material|SetVectorParameterValue _returnvalue_7 "Color" (Math|Color|MakeColor _r _g _b _a) :!class /Script/Engine.MaterialInstanceDynamic))
          (:else))
        (else
          (Components|DestroyComponent (Variables|Default|GetAnimatedLightFixture))
          (Variables|Mesh|SetStaticMesh 0)
          (Components|StaticMesh|SetStaticMesh (Variables|Debug|GetStaticMeshComponent) 0))))
    (:then_1
      (Variables|Default|SetEnd%20Radius (+ (* (Variables|ConeShape|GetCone%20Length) (Math|Trig|Tan(Degrees) (Variables|ConeShape|GetCone%20Angle))) (Variables|ConeShape|GetStart%20Radius)))
      (Transformation|SetWorldScale3D (Variables|Default|GetCone) (Math|Vector|MakeVector (/ (* _x (Variables|Default|GetEnd%20Radius)) 2e+02) (/ (* _y (Variables|Default|GetEnd%20Radius)) 2e+02) (/ (Variables|ConeShape|GetCone%20Length) 1e+03)))
      (Utilities|FlowControl|Sequence
        (:then_0
          (bind _returnvalue_8 (Rendering|Material|CreateDynamicMaterialInstance (Variables|Default|GetCone) 0 (Variables|Default|GetMaterial) "None" :!class /Script/Engine.PrimitiveComponent))
          (Variables|Default|SetMID_Mat _returnvalue_8)
          (Rendering|Material|SetMaterial (Variables|Default|GetCone) 0 _mid_mat)
          (Rendering|Material|SetScalarParameterValue _mid_mat "StartRadius" _start_radius :!class /Script/Engine.MaterialInstanceDynamic)
          (Rendering|Material|SetScalarParameterValue _mid_mat "Additonal_Falloff" _falloff :!class /Script/Engine.MaterialInstanceDynamic)
          (Rendering|Material|SetScalarParameterValue _mid_mat "Profile_Axis_Contribution" _profile_side_contribution :!class /Script/Engine.MaterialInstanceDynamic)
          (Rendering|Material|SetScalarParameterValue _mid_mat "Profile_Brightness" _profile_ground_brightness :!class /Script/Engine.MaterialInstanceDynamic)
          (Rendering|Material|SetScalarParameterValue _mid_mat "WPOAnim_Speed" _returnvalue_6 :!class /Script/Engine.MaterialInstanceDynamic)
          (Rendering|Material|SetScalarParameterValue _mid_mat "WPOAnim_MaxAngle" _returnvalue_5 :!class /Script/Engine.MaterialInstanceDynamic)
          (Rendering|Material|SetScalarParameterValue _mid_mat "IESProfileEdgeFalloff" _profile_edgefalloff :!class /Script/Engine.MaterialInstanceDynamic)
          (Rendering|Material|SetScalarParameterValue _mid_mat "RandomValue" (Variables|Default|GetRandomValue) :!class /Script/Engine.MaterialInstanceDynamic)
          (Rendering|Material|SetVectorParameterValue _mid_mat "Color" (Math|Color|MakeColor _r_3 _g_3 _b_3 _returnvalue_3) :!class /Script/Engine.MaterialInstanceDynamic)
          (Rendering|Material|SetScalarParameterValue _mid_mat "LightShift" _light_shift_2 :!class /Script/Engine.MaterialInstanceDynamic)
          (Rendering|Material|SetScalarParameterValue _mid_mat_2 "Morning" _brightness_morning :!class /Script/Engine.MaterialInstanceDynamic)
          (Rendering|Material|SetScalarParameterValue _mid_mat_2 "Day" _brightness_day :!class /Script/Engine.MaterialInstanceDynamic)
          (Rendering|Material|SetScalarParameterValue _mid_mat_2 "Evening" _brightness_evening :!class /Script/Engine.MaterialInstanceDynamic)
          (Rendering|Material|SetScalarParameterValue _mid_mat_2 "Night" _brightness_night :!class /Script/Engine.MaterialInstanceDynamic))
        (:then_1
          (bind _hlodmaterial (Variables|Default|GetHLODMaterial)
            (:then
              (Rendering|Material|SetMaterial (Variables|Default|GetCone_ForHLOD) 0 _hlodmaterial))
            (:else)))
        (:then_2
          (Rendering|Material|SetDefaultCustomPrimitiveDataFloat (Variables|Default|GetCone_ForHLOD) 4 _start_radius)
          (Rendering|Material|SetDefaultCustomPrimitiveDataFloat (Variables|Default|GetCone_ForHLOD) 5 _falloff)
          (Rendering|Material|SetDefaultCustomPrimitiveDataFloat (Variables|Default|GetCone_ForHLOD) 14 _profile_side_contribution)
          (Rendering|Material|SetDefaultCustomPrimitiveDataFloat (Variables|Default|GetCone_ForHLOD) 15 _profile_ground_brightness)
          (Rendering|Material|SetDefaultCustomPrimitiveDataFloat (Variables|Default|GetCone_ForHLOD) 6 _returnvalue_6)
          (Rendering|Material|SetDefaultCustomPrimitiveDataFloat (Variables|Default|GetCone_ForHLOD) 7 (Math|Float|SelectFloat _returnvalue_5 0.0 (Variables|Default|GetIsAnimated)))
          (Rendering|Material|SetDefaultCustomPrimitiveDataFloat (Variables|Default|GetCone_ForHLOD) 13 _profile_edgefalloff)
          (Rendering|Material|SetDefaultCustomPrimitiveDataFloat (Variables|Default|GetCone_ForHLOD) 3 _returnvalue_3)
          (Rendering|Material|SetDefaultCustomPrimitiveDataVector3 (Variables|Default|GetCone_ForHLOD) 0 (Math|Vector|MakeVector _r_3 _g_3 _b_3))
          (Rendering|Material|SetDefaultCustomPrimitiveDataFloat (Variables|Default|GetCone_ForHLOD) 8 _light_shift_2)
          (Rendering|Material|SetDefaultCustomPrimitiveDataFloat (Variables|Default|GetCone_ForHLOD) 9 _brightness_morning)
          (Rendering|Material|SetDefaultCustomPrimitiveDataFloat (Variables|Default|GetCone_ForHLOD) 10 _brightness_day)
          (Rendering|Material|SetDefaultCustomPrimitiveDataFloat (Variables|Default|GetCone_ForHLOD) 11 _brightness_evening)
          (Rendering|Material|SetDefaultCustomPrimitiveDataFloat (Variables|Default|GetCone_ForHLOD) 12 _brightness_night)
          (Rendering|Material|SetDefaultCustomPrimitiveDataFloat (Variables|Default|GetCone_ForHLOD) 16 (Variables|Default|GetRandomValue)))
        (:then_3
          ;[1 0.7695 0.0498] Control Spawned Lights
          (Rendering|Lighting|SetInnerConeAngle _spotlight (* _cone_angle 0.75))
          (Rendering|Lighting|SetOuterConeAngle _spotlight (* _cone_angle 1.25))
          (Rendering|Components|Light|SetLightColor _spotlight _light_color_rgb__and_brightness_a true)
          (Rendering|Components|Light|SetIntensity _spotlight (* _returnvalue_2 1e+01))
          (Rendering|Components|Light|SetIntensity (Variables|Default|GetPointLight-Scatter) _returnvalue_2)
          (Rendering|Lighting|SetAttenuationRadius (Variables|Default|GetPointLight-Surface) (Variables|ConeShape|GetStart%20Radius))
          (Utilities|FlowControl|Sequence
            (:then_0
              (if (not (Variables|ActualLight|GetSurface%20Point%20Light))
                (Components|DestroyComponent (Variables|Default|GetPointLight-Surface))))
            (:then_1
              (if (not (Variables|ActualLight|GetSpot%20Light))
                (Components|DestroyComponent (Variables|Default|GetSpotLight))))
            (:then_2
              (if (not (Variables|ActualLight|GetScatter%20Light))
                (Components|DestroyComponent (Variables|Default|GetPointLight-Scatter))))))))))
