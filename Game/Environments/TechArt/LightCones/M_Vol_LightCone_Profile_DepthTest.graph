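# Translucent unlit light-cone material that fakes volumetric spotlight beams, with time-of-day intensity and per-primitive or per-instance data.
# A ray-cone intersection bounds an analytic inverse-square scattering integral, optionally modulated by noise or a raymarch with shadow mask.
# Schlick phase and curve-atlas IES profiles, along the axis and projected onto scene surfaces, shape emissive.
# World position offset fits the cone and optionally sways it; quality switches simplify cheaper tiers.
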Material /Game/Environments/TechArt/LightCones/M_Vol_LightCone_Profile_DepthTest
graph {
 properties {
  bAutomaticallySetUsageInEditor=false
  bEnableResponsiveAA=true
  bUsedWithInstancedStaticMeshes=true
  blendMode=BLEND_Translucent
  pixelDepthOffsetMode=PDOM_Legacy
  shadingModel=MSM_Unlit
  usedShadingModels=Unlit
 }
 nodes [
  n1 = Divide(A=ComponentMask(None=NamedRerouteUsage()) {g=true, r=true}, B=NamedRerouteUsage())
  n2 = ComponentMask(None=NamedRerouteUsage()) {g=true, r=true}
  n3 = NamedRerouteDeclaration(None=DotProduct(A=Normalize(VectorInput=@n1), B=Normalize(VectorInput=@n2))) {name=2DCosine, nodeColor=<rgba 0.2188 0 1 1>}
  Abs(None=@n3)
  Constant2Vector() {g=1, r=1}
  Constant2Vector() {g=5, r=5}
  Constant()
  Constant() {r=1}
  MaterialExpressionNamedRerouteUsage_15 = NamedRerouteUsage()
  MaterialExpressionNamedRerouteUsage_14 = NamedRerouteUsage()
  n4 = Constant()
  n5 = Multiply(
   A=NamedRerouteUsage()
   B=StaticSwitchParameter(True=NamedRerouteUsage(), False=ScalarParameter() {defaultValue=0.5, parameterName=LightShift}) {parameterName=UseCustomPrimInstanceData}
  )
  MaterialExpressionAdd_1 = Add(
   A=NamedRerouteUsage()
   B=NamedRerouteDeclaration(None=MaterialFunctionCall(X=@n4, Y=@n4, Z=@n5) {materialFunction=&MakeFloat3}.Result) {name=Light Shift, nodeColor=<rgba 1 0.6757 0 1>}
  ) {constB=0}
  MaterialExpressionNamedRerouteUsage_9 = NamedRerouteUsage()
  MaterialExpressionScalarParameter_6 = ScalarParameter() {defaultValue=0.001, desc=Tiny offset to prevent huge brightness values, group=Advanced, parameterName=TinyOffset}
  Custom!(
   t0=@MaterialExpressionNamedRerouteUsage_15
   t1=@MaterialExpressionNamedRerouteUsage_14
   RayOrigin=@MaterialExpressionAdd_1
   RayDirection=@MaterialExpressionNamedRerouteUsage_9
   f=@MaterialExpressionScalarParameter_6
  ) {
   code="float b = dot(RayOrigin,RayDirection);\r\nfloat c = dot(RayOrigin,RayOrigin);\r\n\r\nt0 = max( t0, 0.0f );\r\n\r\n//tiny offset param to avoid infinite brightness\r\nc+= f;\r\n\r\nfloat h2 = sqrt( (c) - (b*b) );\r\nfloat integral_t0 = ( atanFast( ( t0 + b ) / h2 ) ) / h2;\r\nfloat integral_t1 = ( atanFast( ( t1 + b ) / h2 ) ) / h2;\r\n\r\nreturn (integral_t1 - integral_t0);\r\n"
   description=Inverse Square Falloff
   outputType=CMOT_Float1
   inputs [
    {
     inputName=t0
     input {expression=&self:MaterialExpressionNamedRerouteUsage_15}
    }
    {
     inputName=t1
     input {expression=&self:MaterialExpressionNamedRerouteUsage_14}
    }
    {inputName=RayOrigin, input {expression=&self:MaterialExpressionAdd_1}}
    {
     inputName=RayDirection
     input {expression=&self:MaterialExpressionNamedRerouteUsage_9}
    }
    {inputName=f, input {expression=&self:MaterialExpressionScalarParameter_6}}
   ]
  }
  n6 = ScalarParameter() {defaultValue=4, group=Advanced, parameterName=Fresnel Fade Distance Offset}
  Divide(A=@n6, B=NamedRerouteUsage())
  n7 = NamedRerouteUsage()
  n8 = CrossProduct(A=@n7, B=NamedRerouteUsage())
  DotProduct(A=NamedRerouteUsage(), B=Normalize(VectorInput=CrossProduct(A=@n7, B=@n8)))
  n9 = NamedRerouteUsage()
  n10 = MaterialFunctionCall(Float3=Frac(None=Divide(A=@n9) {constB=100})) {materialFunction=&BreakOutFloat3Components}
  n11 = Add(A=Add(A=@n10.R, B=@n10.R), B=@n10.B)
  Multiply(A=Frac(None=Abs(None=@n11))) {constB=33.3}
  n12 = VectorParameter() {parameterName=Pan Direction and Speed, defaultValue=<rgba 0.25 0 -1 0>}
  n13 = MaterialFunctionCall(Speed=@n12.A) {materialFunction=&TimeWithSpeedVariable}
  NamedRerouteDeclaration(None=VertexInterpolator(VS=Multiply(A=@n12.RGB, B=@n13.Time)).PS) {name=PanVector, nodeColor=<rgba 0 0.9062 1 1>}
  n14 = Normalize(VectorInput=Transform(None=Constant3Vector() {constant=<rgba 0 0 1 0>}) {transformSourceType=TRANSFORMSOURCE_Instance})
  n15 = ShaderStageSwitch(PixelShader=VertexInterpolator(VS=@n14).PS, VertexShader=@n14)
  n16 = StaticSwitchParameter(True=NamedRerouteUsage(), False=@n15) {parameterName=UseWpoAnimation}
  NamedRerouteDeclaration(None=QualitySwitch(Default=ShadingPathSwitch(Default=@n16, Mobile=@n15) {desc=Migrated from FeatureLevelSwitch}, Low=@n15, High=@n15, Medium=@n15)) {name=Cone Z Axis, nodeColor=<rgba 0.0638 0.4605 1 1>}
  n17 = Divide(B=NamedRerouteUsage())
  n18 = NamedRerouteUsage()
  n19 = MaterialFunctionCall!(
   Ray Origin
   Ray Direction
   Cone Origin=NamedRerouteUsage()
   Cone Z Axis=NamedRerouteUsage()
   Cone X Axis=NamedRerouteUsage()
   Cone Y Axis=NamedRerouteUsage()
   Z Min-Max=AppendVector(A=NamedRerouteUsage(), B=NamedRerouteUsage())
   XYZ Scale=AppendVector(A=Divide(A=@n17, B=@n18), B=Constant() {r=1})
  ) {materialFunction=/Volumetrics/Content/Primitives/Lightcone/ConeIntersection}
  NamedRerouteDeclaration(None=@n19.T1) {name=T1, nodeColor=<rgba 0 0 0 1>}
  n20 = LinearInterpolate(A=Power(Base=Multiply(A=@n3) {constB=-1}, Exp=ScalarParameter() {defaultValue=2, parameterName=EdgeFalloffPower}), Alpha=NamedRerouteUsage()) {constAlpha=1}
  NamedRerouteDeclaration(None=StaticSwitchParameter(True=NamedRerouteDeclaration(None=@n20) {name=EdgeFalloff Mobile, nodeColor=<rgba 0 1 0.8594 1>}, False=Constant() {r=1}) {defaultValue=true, parameterName=UseFresnelFalloff}) {name=Edge Falloff, nodeColor=<rgba 0 0 0 1>}
  n21 = NamedRerouteUsage()
  n22 = Abs(None=Add(A=ComponentMask(None=@n21) {b=true}, B=NamedRerouteUsage()))
  n23 = Divide(A=Divide(A=ComponentMask(None=@n21) {g=true, r=true}, B=NamedRerouteUsage()), B=NamedRerouteUsage())
  n24 = MaterialFunctionCall(Vector 3=@n23, Vector 2=@n23) {materialFunction=/Engine/Functions/Engine_MaterialFunctions02/Utility/VectorLength}
  n25 = Divide(
   A=Add(A=Subtract(A=@n22, B=@n24.V2 Length), B=@n6)
   B=Divide(A=ScalarParameter() {defaultValue=4, group=Advanced, parameterName=Fresnel Fade Distance}, B=NamedRerouteUsage())
  )
  n26 = Max(A=ComponentMask(None=NamedRerouteUsage()) {b=true}) {constB=0}
  n27 = ScalarParameter() {defaultValue=4, group=Advanced, parameterName=Underside Fresnel Mask}
  NamedRerouteDeclaration(None=LinearInterpolate(A=Saturate(None=@n25), Alpha=Power(Base=@n26, Exp=@n27) {desc=Prevent Fresnel Falloff from adding artifacts to bottom view.})) {name=Internal Mask, nodeColor=<rgba 0 0.6719 1 1>}
  n28 = Add(A=NamedRerouteUsage(), B=NamedRerouteUsage())
  n29 = ScalarParameter() {defaultValue=256, parameterName=Noise Scale}
  NamedRerouteDeclaration(None=Add(A=Divide(A=@n28, B=@n29), B=NamedRerouteUsage())) {name=Noise Position, nodeColor=<rgba 0 0 0 1>}
  MaterialExpressionMaterialFunctionCall_8 = MaterialFunctionCall() {materialFunction=&ObjectScale}
  MaterialExpressionNamedRerouteUsage_95 = NamedRerouteUsage()
  n30 = Custom!(Scale=@MaterialExpressionMaterialFunctionCall_8.Scale XYZ, StartRadius=@MaterialExpressionNamedRerouteUsage_95) {
   code="float MaxRadius = max(Scale.x, Scale.y)*200;\r\nfloat2 Aspect = Scale.x > Scale.y ? float2(1.0, Scale.y / Scale.x) : float2(Scale.x / Scale.y, 1.0);\r\nfloat ConeLength = Scale.z*1000;\r\n\r\n\r\nfloat ConeTangent = (MaxRadius - StartRadius) / ConeLength;\r\nfloat ConeShift = StartRadius / ConeTangent;\r\n\r\nConeSetup = float3(ConeTangent, ConeShift, ConeLength + ConeShift);\r\n\r\nreturn float4(MaxRadius, ConeLength, Aspect);\r\n\r\n"
   description=Cone Scale and Aspect
   outputType=CMOT_Float4
   additionalOutputs [{outputName=ConeSetup, outputType=CMOT_Float3}]
   inputs [
    {
     inputName=Scale
     input {expression=&self:MaterialExpressionMaterialFunctionCall_8}
    }
    {
     inputName=StartRadius
     input {expression=&self:MaterialExpressionNamedRerouteUsage_95}
    }
   ]
  }
  n31 = ShaderStageSwitch(PixelShader=VertexInterpolator(VS=@n30.return).PS, VertexShader=@n30.return)
  NamedRerouteDeclaration(None=ComponentMask(None=@n31) {r=true}) {name=End Radius, nodeColor=<rgba 0 0.3191 1 1>}
  NamedRerouteDeclaration(None=@n19.CapMask) {name=CapMask, nodeColor=<rgba 0 0 0 1>}
  NamedRerouteDeclaration(None=ComponentMask(None=@n31) {g=true}) {name=Cone Length, nodeColor=<rgba 0 0.3191 1 1>}
  NamedRerouteDeclaration(None=ComponentMask(None=@n31) {a=true, b=true}) {name=Cone Aspect, nodeColor=<rgba 0 0.3191 1 1>}
  n32 = Multiply(
   A=MaterialFunctionCall(Alpha Threshold=ScalarParameter() {defaultValue=0.5, parameterName=Dither Threshold}) {materialFunction=&DitherTemporalAA}.Result
   B=ScalarParameter() {defaultValue=1, parameterName=Dither Amount}
  )
  NamedRerouteDeclaration(None=Divide(A=@n32) {constB=4}) {name=Dither, nodeColor=<rgba 1 0.6094 0 1>}
  NamedRerouteDeclaration(None=@n19.Ray Entry World Position) {name=Cone Ray Entry, nodeColor=<rgba 0 0 0 1>}
  n33 = ShaderStageSwitch(PixelShader=VertexInterpolator(VS=@n30.ConeSetup).PS, VertexShader=@n30.ConeSetup)
  NamedRerouteDeclaration(None=ComponentMask(None=@n33) {r=true}) {name=Cone Tangent, nodeColor=<rgba 0.0078 1 0 1>}
  MaterialExpressionReroute_10 = Reroute(None=TextureObjectParameter() {parameterName=NoiseVolumeTexture, texture=/Engine/EngineSky/VolumetricClouds/T_VolumeNoiseErosion32})
  MaterialExpressionTextureObjectParameter_1 = TextureObjectParameter() {parameterName=ShadowMaskTexture, texture=/Engine/ArtTools/RenderToTexture/Textures/T_EV_BlankWhite_01}
  MaterialExpressionScalarParameter_35 = ScalarParameter() {defaultValue=32, parameterName=Steps}
  MaterialExpressionReroute_5 = Reroute(None=AppendVector(A=NamedRerouteUsage(), B=NamedRerouteUsage()))
  MaterialExpressionNamedRerouteUsage_63 = NamedRerouteUsage()
  MaterialExpressionMultiply_41 = Multiply(A=CameraVectorWS()) {constB=-1}
  MaterialExpressionReroute_6 = Reroute(None=ScalarParameter() {defaultValue=256, parameterName=Noise Scale})
  MaterialExpressionNamedRerouteUsage_74 = NamedRerouteUsage()
  MaterialExpressionNamedRerouteUsage_65 = NamedRerouteUsage()
  MaterialExpressionAppendVector_6 = AppendVector(A=NamedRerouteUsage(), B=NamedRerouteUsage())
  MaterialExpressionAppendVector_8 = AppendVector(A=NamedRerouteUsage(), B=NamedRerouteUsage())
  MaterialExpressionMultiply_43 = Multiply(
   A=NamedRerouteUsage()
   B=StaticSwitchParameter(True=NamedRerouteUsage(), False=ScalarParameter() {defaultValue=0.5, parameterName=LightShift}) {parameterName=UseCustomPrimInstanceData}
  )
  MaterialExpressionNamedRerouteUsage_61 = NamedRerouteUsage()
  MaterialExpressionNamedRerouteUsage_60 = NamedRerouteUsage()
  MaterialExpressionAppendVector_9 = AppendVector(A=ScalarParameter() {defaultValue=1, parameterName=Noise Contrast}, B=ScalarParameter() {defaultValue=2, parameterName=ShadowMaskContrast})
  n34 = ScalarParameter() {defaultValue=0.5, parameterName=Shadow Mask Pan Speed}
  MaterialExpressionAppendVector_17 = AppendVector(
   A=Add(
    A=MaterialFunctionCall(Speed=@n34) {materialFunction=&TimeWithSpeedVariable}.Time
    B=ScalarParameter() {parameterName=Shadowmask UV Offset}
   )
   B=ScalarParameter() {defaultValue=1, parameterName=ShadowMask UV Scale}
  )
  n35 = NamedRerouteUsage()
  n36 = ScalarParameter() {defaultValue=256, parameterName=Noise Scale}
  MaterialExpressionSubtract_7 = Subtract(A=Add(A=NamedRerouteUsage(), B=Divide(A=@n35, B=@n36)), B=NamedRerouteUsage())
  n37 = Custom!(
   Tex=@MaterialExpressionReroute_10
   ShadowMask=@MaterialExpressionTextureObjectParameter_1
   Steps=@MaterialExpressionScalarParameter_35
   Intersection=@MaterialExpressionReroute_5
   RO=@MaterialExpressionNamedRerouteUsage_63
   RD=@MaterialExpressionMultiply_41
   NoiseScale=@MaterialExpressionReroute_6
   Dither=@MaterialExpressionNamedRerouteUsage_74
   EnableRayMarch=@MaterialExpressionNamedRerouteUsage_65
   ZExtents=@MaterialExpressionAppendVector_6
   Radii=@MaterialExpressionAppendVector_8
   LightShift=@MaterialExpressionMultiply_43
   RO_Local=@MaterialExpressionNamedRerouteUsage_61
   RD_Local=@MaterialExpressionNamedRerouteUsage_60
   Contrasts=@MaterialExpressionAppendVector_9
   ShadowMaskUVBiasScale=@MaterialExpressionAppendVector_17
   Panner=@MaterialExpressionSubtract_7
  ) {
   code="float Accum = 0;\r\n\r\nfloat TotalDistsance = max(0, Intersection.y - Intersection.x);\r\n\r\nfloat StepSize = (TotalDistsance / Steps);\r\nfloat ConeLength = (ZExtents.y - ZExtents.x);\r\nRO += float3(0,0,LightShift);\r\n\r\nfor (int i = 0; i < Steps; i++)\r\n{\r\n\tfloat3 CurUVW = RO + (RD * (i+Dither)* StepSize);\r\n\t\r\n\tfloat CurSample = Tex.SampleLevel(TexSampler, (CurUVW / NoiseScale) + Panner, 0).r;\r\n\tCurSample = pow(CurSample, Contrasts.x);\r\n\t\r\n\tfloat InvSq = 1 / (length(CurUVW) * length(CurUVW));\r\n\t\r\n\t##if RAYMARCHSHADOWMASK\r\n\t\r\n\t\tfloat3 CurUVW_Local = RO_Local + (RD_Local * (i+Dither)* StepSize);\r\n\t\tfloat normalizedZ = (CurUVW_Local.z) / ConeLength;\r\n\t\tfloat CurRadius = lerp(Radii.x, Radii.y, normalizedZ);\r\n\t\tfloat2 MaskUV = ((CurUVW_Local.xy / CurRadius) / 2) + 0.5;\r\n\t\tMaskUV = (MaskUV + ShadowMaskUVBiasScale.x) * ShadowMaskUVBiasScale.y;\r\n\t\t\r\n\t\tfloat CurShadow = ShadowMask.SampleLevel(ShadowMaskSampler, MaskUV, 0);\r\n\t\tCurSample *= pow(CurShadow, Contrasts.y);\r\n\t\r\n\t##endif\r\n\r\n\tAccum += (CurSample * InvSq) * StepSize;\r\n} \r\n\r\n\r\nreturn Accum;\r\n\r\n//return  Tex.SampleLevel(TexSampler, RO / NoiseScale, 0).r;\r\n\t"
   inputs [
    {inputName=Tex, input {expression=&self:MaterialExpressionReroute_10}}
    {
     inputName=ShadowMask
     input {expression=&self:MaterialExpressionTextureObjectParameter_1}
    }
    {
     inputName=Steps
     input {expression=&self:MaterialExpressionScalarParameter_35}
    }
    {
     inputName=Intersection
     input {expression=&self:MaterialExpressionReroute_5}
    }
    {
     inputName=RO
     input {expression=&self:MaterialExpressionNamedRerouteUsage_63}
    }
    {inputName=RD, input {expression=&self:MaterialExpressionMultiply_41}}
    {inputName=NoiseScale, input {expression=&self:MaterialExpressionReroute_6}}
    {
     inputName=Dither
     input {expression=&self:MaterialExpressionNamedRerouteUsage_74}
    }
    {
     inputName=EnableRayMarch
     input {expression=&self:MaterialExpressionNamedRerouteUsage_65}
    }
    {
     inputName=ZExtents
     input {expression=&self:MaterialExpressionAppendVector_6}
    }
    {inputName=Radii, input {expression=&self:MaterialExpressionAppendVector_8}}
    {
     inputName=LightShift
     input {expression=&self:MaterialExpressionMultiply_43}
    }
    {
     inputName=RO_Local
     input {expression=&self:MaterialExpressionNamedRerouteUsage_61}
    }
    {
     inputName=RD_Local
     input {expression=&self:MaterialExpressionNamedRerouteUsage_60}
    }
    {
     inputName=Contrasts
     input {expression=&self:MaterialExpressionAppendVector_9}
    }
    {
     inputName=ShadowMaskUVBiasScale
     input {expression=&self:MaterialExpressionAppendVector_17}
    }
    {inputName=Panner, input {expression=&self:MaterialExpressionSubtract_7}}
   ]
  }
  NamedRerouteDeclaration(None=Multiply(A=@n37, B=ScalarParameter() {defaultValue=1, parameterName=RayMarchMultiplier})) {name=RayMarchedCone, nodeColor=<rgba 1 0 0.6562 1>}
  NamedRerouteDeclaration(None=@n19.Ray Entry Local Position) {name=Ray Entry Local, nodeColor=<rgba 0 0 0 1>}
  NamedRerouteDeclaration(None=ComponentMask(None=@n33) {g=true}) {name=ZMin, nodeColor=<rgba 0.0078 1 0 1>}
  n38 = ComponentMask(None=@n33) {b=true}
  n39 = Multiply(A=@n38, B=ScalarParameter() {defaultValue=0.376, parameterName=LowDetailDownScale})
  NamedRerouteDeclaration(None=StaticSwitchParameter(True=QualitySwitch(Default=@n38, Low=@n39, High=@n39, Medium=@n39), False=@n38) {defaultValue=true, parameterName=UseLowDetailDownScale}) {name=ZMax, nodeColor=<rgba 0.0078 1 0 1>}
  n40 = WorldPosition()
  n41 = Length(None=Subtract(A=@n40.XYZ, B=NamedRerouteUsage())) {is=class MaterialExpressionLength}
  n42 = NamedRerouteUsage()
  n43 = SquareRoot(None=Subtract(A=Multiply(A=@n41, B=@n41), B=Multiply(A=@n42, B=@n42)))
  n44 = NamedRerouteUsage()
  n45 = Subtract(A=@n40.XYZ, B=Add(A=Multiply(A=@n43, B=@n44), B=NamedRerouteUsage()))
  NamedRerouteDeclaration(None=Multiply(B=Normalize(VectorInput=@n45)) {constA=-1}) {name=Direction From Axis, nodeColor=<rgba 1 0 0.7266 1>}
  NamedRerouteDeclaration(None=StaticSwitchParameter(True=NamedRerouteUsage(), False=ScalarParameter() {defaultValue=25, parameterName=StartRadius}) {parameterName=UseCustomPrimInstanceData}) {name=Start radius, nodeColor=<rgba 0 0.3191 1 1>}
  n46 = NamedRerouteUsage()
  n47 = CameraPositionWS()
  NamedRerouteDeclaration(None=Normalize(VectorInput=Subtract(A=@n46, B=@n47))) {name=CameraVector To Origin, nodeColor=<rgba 0 1 0.9062 1>}
  NamedRerouteDeclaration(None=StaticSwitchParameter(
    True=Custom!() {code="##define RAYMARCHSHADOWMASK 1\r\nreturn 0;", description="##define RAYMARCHSHADOWMASK 1", inputs []}
    False=Custom() {code="##define RAYMARCHSHADOWMASK 0\r\nreturn 0;", description="##define RAYMARCHSHADOWMASK 0", inputs []}
   ) {parameterName=RayMarchShadowMask}) {name=ShadowMask Define, nodeColor=<rgba 0 0 0 1>}
  n48 = Normalize(VectorInput=Transform(None=Constant3Vector() {constant=<rgba 1 0 0 0>}) {transformSourceType=TRANSFORMSOURCE_Instance})
  n49 = ShaderStageSwitch(PixelShader=VertexInterpolator(VS=@n48).PS, VertexShader=@n48)
  n50 = StaticSwitchParameter(True=NamedRerouteUsage(), False=@n49) {parameterName=UseWpoAnimation}
  NamedRerouteDeclaration(None=QualitySwitch(Default=ShadingPathSwitch(Default=@n50, Mobile=@n49) {desc=Migrated from FeatureLevelSwitch}, Low=@n49, High=@n49, Medium=@n49)) {name=Cone X Axis, nodeColor=<rgba 0.0638 0.4605 1 1>}
  NamedRerouteDeclaration(None=@n19.Ray Exit World Position) {name=Cone Ray Exit, nodeColor=<rgba 0 0 0 1>}
  n51 = Normalize(VectorInput=Transform(None=Constant3Vector() {constant=<rgba 0 1 0 0>}) {transformSourceType=TRANSFORMSOURCE_Instance})
  n52 = ShaderStageSwitch(PixelShader=VertexInterpolator(VS=@n51).PS, VertexShader=@n51)
  n53 = StaticSwitchParameter(True=NamedRerouteUsage(), False=@n52) {parameterName=UseWpoAnimation}
  NamedRerouteDeclaration(None=QualitySwitch(Default=ShadingPathSwitch(Default=@n53, Mobile=@n52) {desc=Migrated from FeatureLevelSwitch}, Low=@n52, High=@n52, Medium=@n52)) {name=Cone Y Axis, nodeColor=<rgba 0.0638 0.4605 1 1>}
  NamedRerouteDeclaration(None=@n32) {name=Dither for Raymarch, nodeColor=<rgba 0 0.8594 1 1>}
  n54 = NamedRerouteUsage()
  n55 = Abs(None=@n54)
  n56 = CurveAtlasRowParameter(CurveTime=Power(Base=OneMinus(None=@n55), Exp=ScalarParameter() {defaultValue=0.4404, group=SpotLight Profile, parameterName=IESProfileMapping})) {
   atlas=&/SpotLightProfiles
   curve=&/SpotLight_Profile_Complex
   group=SpotLight Profile
   parameterName=ProfileSelection
  }
  n57 = NamedRerouteUsage()
  n58 = ScalarParameter() {defaultValue=0.496, group=SpotLight Profile, parameterName=Profile_Axis_Contribution}
  n59 = Saturate(None=Add(A=@n56.R, B=StaticSwitchParameter(True=@n57, False=@n58) {parameterName=UseCustomPrimInstanceData}))
  n60 = VertexColor()
  n61 = Saturate(None=Power(Base=OneMinus(None=@n60.R), Exp=ScalarParameter() {defaultValue=1.3112, group=SpotLight Profile, parameterName=Profile_Axis_Falloff}))
  n62 = DotProduct(A=Transform(None=Constant3Vector() {constant=<rgba 0 0 1 0>}) {transformSourceType=TRANSFORMSOURCE_Camera}, B=NamedRerouteUsage())
  n63 = LinearInterpolate(A=LinearInterpolate(B=@n59, Alpha=@n61) {constA=1}, B=Multiply(A=@n61, B=@n59), Alpha=Saturate(None=Multiply(A=@n62) {constB=-1}))
  n64 = NamedRerouteUsage()
  n65 = Saturate(None=NamedRerouteDeclaration(None=DotProduct(A=@n64, B=@n7)) {name=CamDotDir, nodeColor=<rgba 0.9219 1 0 1>})
  n66 = Multiply(A=@n54) {constB=-1}
  n67 = StaticSwitchParameter(True=NamedRerouteUsage(), False=ScalarParameter() {defaultValue=1.5053, group=SpotLight Profile, parameterName=IESProfileEdgeFalloff}) {parameterName=UseCustomPrimInstanceData}
  n68 = LinearInterpolate(A=Saturate(None=Power(Base=@n66, Exp=@n67)), Alpha=NamedRerouteUsage()) {constAlpha=1}
  n69 = Saturate(None=Multiply(A=@n62) {constB=8})
  NamedRerouteDeclaration(None=Multiply(A=LinearInterpolate(A=@n63, Alpha=@n65) {constA=1}, B=LinearInterpolate(A=@n68, Alpha=@n69) {constA=1})) {name=IES Profile Along Axis, nodeColor=<rgba 0.7344 0 1 1>}
  n70 = Subtract(
   A=MaterialFunctionCall(Exclude WPO Offsets=StaticBool()) {materialFunction=&WorldPositionBehindTranslucency}.Result
   B=NamedRerouteUsage()
  )
  n71 = NamedRerouteUsage()
  n72 = DotProduct(A=@n70, B=@n71)
  n73 = NamedRerouteUsage()
  n74 = Divide(A=Length(None=CrossProduct(A=@n70, B=@n71)), B=LinearInterpolate(A=NamedRerouteUsage(), B=NamedRerouteUsage(), Alpha=Divide(A=@n72, B=@n73)))
  n75 = CurveAtlasRowParameter(CurveTime=NamedRerouteDeclaration(None=Saturate(None=@n74)) {name=Distance From Axis, nodeColor=<rgba 0.6875 0 1 1>}) {
   atlas=&SpotLightProfiles
   curve=&SpotLight_Profile_Complex
   group=SpotLight Profile
   parameterName=ProfileSelection
  }
  n76 = Multiply(
   A=Multiply(A=Saturate(None=@n72), B=@n75.R)
   B=StaticSwitchParameter(True=NamedRerouteUsage(), False=ScalarParameter() {defaultValue=0.088, group=SpotLight Profile, parameterName=Profile_Brightness}) {parameterName=UseCustomPrimInstanceData}
  )
  n77 = DotProduct(A=ComponentMask(None=SceneTexture() {sceneTextureId=PPI_WorldNormal}.Color) {b=true, g=true, r=true}, B=NamedRerouteUsage())
  n78 = Multiply(A=Saturate(None=Multiply(A=@n77) {constB=-1}), B=@n76)
  NamedRerouteDeclaration(None=QualitySwitch(Default=@n76, Epic=@n78)) {name=IES Profile Ground Projection, nodeColor=<rgba 0 1 0.4609 1>}
  NamedRerouteDeclaration(None=@n19.Ray Exit Local Position - NoCaps) {name=Ray Exit Local-NoCaps, nodeColor=<rgba 0 0 0 1>}
  n79 = NamedRerouteUsage()
  n80 = RotateAboutAxis(NormalizedRotationAxis=NamedRerouteUsage(), RotationAngle=NamedRerouteUsage(), PivotPoint=Constant3Vector(), Position=@n79)
  NamedRerouteDeclaration(None=Normalize(VectorInput=Add(A=@n80, B=@n79))) {name=Cone Z Rotated, nodeColor=<rgba 1 0 0.375 1>}
  n81 = NamedRerouteUsage()
  n82 = NamedRerouteUsage()
  NamedRerouteDeclaration(None=Normalize(VectorInput=CrossProduct(A=@n81, B=@n82))) {name=Cone X Rotated, nodeColor=<rgba 1 0 0.375 1>}
  n83 = NamedRerouteUsage()
  n84 = NamedRerouteUsage()
  NamedRerouteDeclaration(None=Normalize(VectorInput=CrossProduct(A=@n83, B=@n84))) {name=Cone Y Rotated, nodeColor=<rgba 1 0 0.375 1>}
  NamedRerouteDeclaration(None=@n19.Local Ray Origin) {name=Ray Origin Local, nodeColor=<rgba 0 0 0 1>}
  n85 = Normalize(VectorInput=Transform(None=Constant3Vector() {constant=<rgba 1 0 0 0>}) {transformSourceType=TRANSFORMSOURCE_Instance})
  NamedRerouteDeclaration(None=ShaderStageSwitch(PixelShader=VertexInterpolator(VS=@n85).PS, VertexShader=@n85)) {name=Cone X BeforeRotation, nodeColor=<rgba 1 0.6797 0 1>}
  n86 = Time()
  n87 = StaticSwitchParameter(True=NamedRerouteUsage(), False=ScalarParameter() {parameterName=RandomValue}) {parameterName=UseCustomPrimInstanceData}
  n88 = NamedRerouteUsage()
  n89 = ScalarParameter() {defaultValue=0.1, parameterName=WPOAnim_Speed}
  n90 = Sine(None=Multiply(A=Add(A=@n86, B=@n87), B=StaticSwitchParameter(True=@n88, False=@n89) {parameterName=UseCustomPrimInstanceData}))
  n91 = StaticSwitchParameter(True=NamedRerouteUsage(), False=ScalarParameter() {defaultValue=0.1, parameterName=WPOAnim_MaxAngle}) {parameterName=UseCustomPrimInstanceData}
  NamedRerouteDeclaration(None=Frac(None=Multiply(A=@n90, B=@n91))) {name=Anim Rotation, nodeColor=<rgba 0 1 0.5078 1>}
  n92 = Normalize(VectorInput=Transform(None=Constant3Vector() {constant=<rgba 0 1 0 0>}) {transformSourceType=TRANSFORMSOURCE_Instance})
  n93 = VertexInterpolator(VS=@n92)
  NamedRerouteDeclaration(None=NamedRerouteDeclaration(None=ShaderStageSwitch(PixelShader=@n93.PS, VertexShader=@n92)) {name=Cone Y BeforeRotation, nodeColor=<rgba 1 0.6797 0 1>}) {name=RotationAxis, nodeColor=<rgba 0.0312 1 0 1>}
  n94 = Normalize(VectorInput=Transform(None=Constant3Vector() {constant=<rgba 0 0 1 0>}) {transformSourceType=TRANSFORMSOURCE_Instance})
  NamedRerouteDeclaration(None=ShaderStageSwitch(PixelShader=VertexInterpolator(VS=@n94).PS, VertexShader=@n94)) {name=Cone Z BeforeRotation, nodeColor=<rgba 1 0.6797 0 1>}
  NamedRerouteDeclaration(None=@n62) {name=CamDotZAXis, nodeColor=<rgba 0.9688 0 1 1>}
  n95 = VectorParameter() {bUseCustomPrimitiveData=true, group=PrimitiveData, parameterName=Prim_Color, defaultValue=<rgba 1 1 1 1>}
  n96 = PerInstanceCustomData3Vector!(DefaultValue) {constDefaultValue=<rgba 1 1 1 0>}
  n97 = PerInstanceCustomData!(DefaultValue=Constant()) {dataIndex=3}
  NamedRerouteDeclaration(None=LinearInterpolate(A=Multiply(A=@n95.RGB, B=@n95.RGB), B=Multiply(A=@n96, B=@n97), Alpha=NamedRerouteUsage())) {name=PrimDataColor, nodeColor=<rgba 0.6172 1 0 1>}
  n98 = Saturate(None=Ceil(None=@n97))
  NamedRerouteDeclaration(None=@n98) {name=IsInstancePrimData, nodeColor=<rgba 0.4531 1 0 1>}
  NamedRerouteDeclaration(None=@n19.Local Ray Direction) {name=Ray Direction Local, nodeColor=<rgba 0 0 0 1>}
  NamedRerouteDeclaration(None=LinearInterpolate(
    A=ScalarParameter() {bUseCustomPrimitiveData=true, defaultValue=5, parameterName=Prim_StartRadius, primitiveDataIndex=4}
    B=PerInstanceCustomData() {constDefaultValue=25, dataIndex=4}
    Alpha=NamedRerouteUsage()
   )) {name=PrimStartRadius, nodeColor=<rgba 0.6172 1 0 1>}
  NamedRerouteDeclaration(None=LinearInterpolate(
    A=ScalarParameter() {bUseCustomPrimitiveData=true, parameterName=Prim_LightShift, primitiveDataIndex=8}
    B=PerInstanceCustomData() {dataIndex=8}
    Alpha=NamedRerouteUsage()
   )) {name=PrimLightShift, nodeColor=<rgba 0.6172 1 0 1>}
  NamedRerouteDeclaration(None=LinearInterpolate(
    A=ScalarParameter() {bUseCustomPrimitiveData=true, defaultValue=5, parameterName=Prim_AdditonalFalloff, primitiveDataIndex=5}
    B=PerInstanceCustomData() {constDefaultValue=1, dataIndex=5}
    Alpha=NamedRerouteUsage()
   )) {name=PrimAdditonalFalloff, nodeColor=<rgba 0.6172 1 0 1>}
  NamedRerouteDeclaration(None=LinearInterpolate(
    A=ScalarParameter() {bUseCustomPrimitiveData=true, parameterName=Prim_Morning, primitiveDataIndex=9}
    B=PerInstanceCustomData() {constDefaultValue=1, dataIndex=9}
    Alpha=NamedRerouteUsage()
   )) {name=PrimMorning, nodeColor=<rgba 0.6172 1 0 1>}
  NamedRerouteDeclaration(None=LinearInterpolate(
    A=ScalarParameter() {bUseCustomPrimitiveData=true, parameterName=Prim_Day, primitiveDataIndex=10}
    B=PerInstanceCustomData() {constDefaultValue=1, dataIndex=10}
    Alpha=NamedRerouteUsage()
   )) {name=PrimDay, nodeColor=<rgba 0.6172 1 0 1>}
  NamedRerouteDeclaration(None=LinearInterpolate(
    A=ScalarParameter() {bUseCustomPrimitiveData=true, parameterName=Prim_Evening, primitiveDataIndex=11}
    B=PerInstanceCustomData() {constDefaultValue=1, dataIndex=11}
    Alpha=NamedRerouteUsage()
   )) {name=PrimEvening, nodeColor=<rgba 0.6172 1 0 1>}
  NamedRerouteDeclaration(None=LinearInterpolate(
    A=ScalarParameter() {bUseCustomPrimitiveData=true, parameterName=Prim_Night, primitiveDataIndex=12}
    B=PerInstanceCustomData() {constDefaultValue=1, dataIndex=12}
    Alpha=NamedRerouteUsage()
   )) {name=PrimNight, nodeColor=<rgba 0.6172 1 0 1>}
  NamedRerouteDeclaration(None=LinearInterpolate(
    A=ScalarParameter() {bUseCustomPrimitiveData=true, parameterName=Prim_ProfileEdgeFalloff, primitiveDataIndex=13}
    B=PerInstanceCustomData() {constDefaultValue=1, dataIndex=13}
    Alpha=NamedRerouteUsage()
   )) {name=PrimProfileEdgeFalloff, nodeColor=<rgba 0.6172 1 0 1>}
  NamedRerouteDeclaration(None=LinearInterpolate(
    A=ScalarParameter() {bUseCustomPrimitiveData=true, parameterName=Prim_Profile_Axis_Contribution, primitiveDataIndex=14}
    B=PerInstanceCustomData() {constDefaultValue=0.496, dataIndex=14}
    Alpha=NamedRerouteUsage()
   )) {name=PrimProfile_Axis_Contribution, nodeColor=<rgba 0.6172 1 0 1>}
  NamedRerouteDeclaration(None=LinearInterpolate(
    A=ScalarParameter() {bUseCustomPrimitiveData=true, parameterName=Prim_Profile_Brightness, primitiveDataIndex=15}
    B=PerInstanceCustomData() {constDefaultValue=0.088, dataIndex=15}
    Alpha=NamedRerouteUsage()
   )) {name=PrimProfile_Brightness, nodeColor=<rgba 0.6172 1 0 1>}
  NamedRerouteDeclaration(None=TransformPosition(None=Constant3Vector()) {transformSourceType=TRANSFORMPOSSOURCE_Instance, transformType=TRANSFORMPOSSOURCE_World}) {name=Cone Origin, nodeColor=<rgba 0.0638 0.4605 1 1>}
  NamedRerouteDeclaration(None=LinearInterpolate(
    A=ScalarParameter() {bUseCustomPrimitiveData=true, parameterName=Prim_WPOAnim_MaxAngle, primitiveDataIndex=7}
    B=PerInstanceCustomData() {dataIndex=7}
    Alpha=NamedRerouteUsage()
   )) {name=PrimWPOAnim_MaxAngle, nodeColor=<rgba 0.6172 1 0 1>}
  NamedRerouteDeclaration(None=LinearInterpolate(
    A=ScalarParameter() {bUseCustomPrimitiveData=true, parameterName=Prim_WPOAnim_Speed, primitiveDataIndex=6}
    B=PerInstanceCustomData() {dataIndex=6}
    Alpha=NamedRerouteUsage()
   )) {name=PrimWPOAnim_Speed, nodeColor=<rgba 0.6172 1 0 1>}
  NamedRerouteDeclaration(None=LinearInterpolate(
    A=ScalarParameter() {bUseCustomPrimitiveData=true, parameterName=Random, primitiveDataIndex=16}
    B=PerInstanceCustomData() {dataIndex=16}
    Alpha=NamedRerouteUsage()
   )) {name=RandomValue, nodeColor=<rgba 0.6172 1 0 1>}
  n99 = MaterialFunctionCall(
   Input=Distance(A=ActorPositionWS(), B=CameraPositionWS())
   Input Low=ScalarParameter() {defaultValue=1000, group=Light Intensity - Distance Fade, parameterName=FadeDistance - Min}
   Input High=ScalarParameter() {defaultValue=2000, group=Light Intensity - Distance Fade, parameterName=FadeDistance - Max}
   Target Low=Constant()
   Target High=Constant() {r=1}
  ) {materialFunction=&RemapValueRange}
  NamedRerouteDeclaration(None=LinearInterpolate(
    A=NamedRerouteUsage()
    B=ScalarParameter() {defaultValue=1, group=Light Intensity - Distance Fade, parameterName=Light Intensity - Close}
    Alpha=Clamp(None=@n99.Result)
   )) {name=Light Intensity - Distance Fade, nodeColor=<rgba 0.0078 0 1 1>}
  NamedRerouteDeclaration(None=@n19.Ray Entry Local Position - NoCaps) {name=Ray Entry Local-NoCaps, nodeColor=<rgba 0 0 0 1>}
  n100 = Add(A=Constant(), B=NamedRerouteUsage())
  n101 = Multiply(
   A=Divide(
    A=ScalarParameter() {defaultValue=256, parameterName=Noise Trace Distance}
    B=ScalarParameter() {defaultValue=256, parameterName=Noise Scale}
   )
   B=Multiply(A=CameraVectorWS()) {constB=-1}
  )
  n102 = TextureSampleParameterVolume!(UVs=Add(A=Multiply(A=@n101, B=@n100), B=NamedRerouteUsage()), Apply View MipBias) {parameterName=NoiseVolumeTexture, samplerType=SAMPLERTYPE_Masks, texture=&T_VolumeNoiseErosion32}
  n103 = Add(A=Constant() {r=0.25}, B=NamedRerouteUsage())
  n104 = TextureSampleParameterVolume(UVs=Add(A=Multiply(A=@n101, B=@n103), B=NamedRerouteUsage())) {parameterName=NoiseVolumeTexture, samplerType=SAMPLERTYPE_Masks, texture=&T_VolumeNoiseErosion32}
  n105 = Add(A=@n102.R, B=@n104.R)
  n106 = Add(A=Constant() {r=0.5}, B=NamedRerouteUsage())
  n107 = TextureSampleParameterVolume(UVs=Add(A=Multiply(A=@n101, B=@n106), B=NamedRerouteUsage())) {parameterName=NoiseVolumeTexture, samplerType=SAMPLERTYPE_Masks, texture=&T_VolumeNoiseErosion32}
  n108 = Add(A=Constant() {r=0.75}, B=NamedRerouteUsage())
  n109 = TextureSampleParameterVolume(UVs=Add(A=Multiply(A=@n101, B=@n108), B=NamedRerouteUsage())) {parameterName=NoiseVolumeTexture, samplerType=SAMPLERTYPE_Masks, texture=&T_VolumeNoiseErosion32}
  n110 = Divide(A=Add(A=Add(A=@n105, B=@n107.R), B=@n109.R)) {constB=4}
  NamedRerouteDeclaration(None=Power(Base=Saturate(None=@n110), Exp=ScalarParameter() {defaultValue=1, parameterName=Noise Contrast})) {name=VolumeNoise, nodeColor=<rgba 0.0335 0.5098 1 1>}
  NamedRerouteDeclaration(None=@n19.T0) {name=T0, nodeColor=<rgba 0 0 0 1>}
  NamedRerouteUsage()
  OneMinus(None=@n54)
  Power(Base=@n75.R, Exp=ScalarParameter() {defaultValue=0.9599, group=SpotLight Profile, parameterName=Profile_Contrast})
  n111 = MaterialFunctionCall(Value_Morning=NamedRerouteUsage(), Value_Day=NamedRerouteUsage(), Value_Evening=NamedRerouteUsage(), Value_Night=NamedRerouteUsage()) {materialFunction=&Scalar_Time_of_Day}
  n112 = MaterialFunctionCall(
   Value_Morning=ScalarParameter() {defaultValue=1, parameterName=Morning}
   Value_Day=ScalarParameter() {defaultValue=1, parameterName=Day}
   Value_Evening=ScalarParameter() {defaultValue=1, parameterName=Evening}
   Value_Night=ScalarParameter() {defaultValue=1, parameterName=Night}
  ) {materialFunction=&Scalar_Time_of_Day}
  n113 = NamedRerouteUsage()
  n114 = CameraVectorWS()
  n115 = MaterialFunctionCall!(
   Cosine=DotProduct(A=Normalize(VectorInput=@n113), B=Multiply(A=@n114) {constB=-1})
   Anisotropy=ScalarParameter() {defaultValue=0.8, parameterName=Anisotropy}
  ) {materialFunction=/Engine/Functions/Engine_MaterialFunctions01/Volumetrics/SchlickPhase}
  n116 = Add(A=NamedRerouteUsage(), B=@n115.Result)
  n117 = QualitySwitch(
   Default=ShadingPathSwitch(
    Default=StaticSwitchParameter(True=@n116, False=@n115.Result) {defaultValue=true, parameterName=UseIES Profile Ground Projection}
    Mobile=@n115.Result
   ) {desc=Migrated from FeatureLevelSwitch}
   Low=@n115.Result
   High=@n115.Result
   Medium=@n115.Result
  )
  n118 = Multiply(A=NamedRerouteUsage(), B=@n117)
  n119 = QualitySwitch(
   Default=ShadingPathSwitch(Default=StaticSwitchParameter(True=@n118, False=@n117) {defaultValue=true, parameterName=UseIES Profile Axis Projection}, Mobile=@n117) {desc=Migrated from FeatureLevelSwitch}
   Low=@n117
   High=@n117
   Medium=@n117
  )
  n120 = VectorParameter() {parameterName=Color, defaultValue=<rgba 0.6 0.75 1 5000>}
  n121 = NamedRerouteUsage()
  n122 = NamedRerouteDeclaration(None=@n120.A) {name=Color - A Channel, nodeColor=<rgba 0 1 0.6484 1>}
  n123 = StaticSwitchParameter(
   True=NamedRerouteUsage()
   False=Multiply(A=@n120.RGB, B=StaticSwitchParameter(True=@n121, False=@n122) {parameterName=Light Intensity - Distance Fade?})
  ) {parameterName=UseCustomPrimInstanceData}
  n124 = Multiply(A=@n123, B=ScalarParameter() {defaultValue=0.376, parameterName=LowDetailDownScale})
  n125 = StaticSwitchParameter(True=QualitySwitch(Default=@n123, Low=@n124, High=@n124, Medium=@n124), False=@n123) {defaultValue=true, parameterName=UseLowDetailDownScale}
  Reroute(None=Multiply(A=StaticSwitchParameter(True=@n111.Result, False=@n112.Result) {parameterName=UseCustomPrimInstanceData}, B=Multiply(A=@n119, B=@n125))) {desc=Emissive}
  n126 = ScalarParameter() {defaultValue=0.1, parameterName=Brightness Looking at cone}
  n127 = Saturate(None=Multiply(A=NamedRerouteUsage()) {constB=-1})
  n128 = NamedRerouteUsage()
  n129 = NamedRerouteUsage()
  n130 = ShadingPathSwitch(Default=@n129) {desc=Migrated from FeatureLevelSwitch}
  n131 = Constant() {r=1}
  n132 = StaticSwitchParameter(True=NamedRerouteUsage(), False=ScalarParameter() {defaultValue=3, parameterName=Additonal_Falloff}) {parameterName=UseCustomPrimInstanceData}
  n133 = Multiply(A=@n132, B=ScalarParameter() {defaultValue=1.106, parameterName=LowDetailDownScale_Falloff})
  MaterialExpressionLinearInterpolate_9 = LinearInterpolate(
   B=StaticSwitchParameter(True=QualitySwitch(Default=@n132, Low=@n133, High=@n133, Medium=@n133), False=@n132) {defaultValue=true, parameterName=UseLowDetailDownScale}
   Alpha=Power(Base=VertexColor().R, Exp=ScalarParameter() {defaultValue=1, parameterName=VertexColorPower})
  ) {constA=1, constB=0}
  n134 = Custom!(
   t0=@MaterialExpressionNamedRerouteUsage_15
   t1=@MaterialExpressionNamedRerouteUsage_14
   RayOrigin=@MaterialExpressionAdd_1
   RayDirection=@MaterialExpressionNamedRerouteUsage_9
   f=@MaterialExpressionScalarParameter_6
   falloff=@MaterialExpressionLinearInterpolate_9
  ) {
   code="float b = dot(RayOrigin,RayDirection);\r\nfloat c = dot(RayOrigin,RayOrigin);\r\n\r\nt0 = max( t0, 0.0f );\r\n\r\n//tiny offset param to avoid infinite brightness\r\nc+= f;\r\n\r\nfloat h2 = sqrt( (c) - (b*b) );\r\nh2 = pow(h2,falloff);\r\nfloat integral_t0 = ( atanFast( ( t0 + b ) / h2 ) ) / h2;\r\nfloat integral_t1 = ( atanFast( ( t1 + b ) / h2 ) ) / h2;\r\n\r\nreturn (integral_t1 - integral_t0);\r\n"
   description=Inverse Square Falloff
   outputType=CMOT_Float1
   inputs [
    {
     inputName=t0
     input {expression=&self:MaterialExpressionNamedRerouteUsage_15}
    }
    {
     inputName=t1
     input {expression=&self:MaterialExpressionNamedRerouteUsage_14}
    }
    {inputName=RayOrigin, input {expression=&self:MaterialExpressionAdd_1}}
    {
     inputName=RayDirection
     input {expression=&self:MaterialExpressionNamedRerouteUsage_9}
    }
    {inputName=f, input {expression=&self:MaterialExpressionScalarParameter_6}}
    {
     inputName=falloff
     input {expression=&self:MaterialExpressionLinearInterpolate_9}
    }
   ]
  }
  n135 = ComponentMask(None=@n134) {r=true}
  n136 = Multiply(A=@n135, B=NamedRerouteUsage())
  n137 = Multiply(
   A=StaticSwitchParameter(True=QualitySwitch(Default=@n130, Low=@n131, High=@n131, Medium=@n131), False=@n129) {parameterName=DisableRayMarchedNoiseOnLow}
   B=@n136
  ) {constA=1, desc=O}
  n138 = Multiply(A=@n135, B=NamedRerouteUsage())
  n139 = ShadingPathSwitch(
   Default=StaticSwitchParameter(True=StaticSwitchParameter(True=@n128, False=@n137) {parameterName=RayMarchedNoise}, False=@n136) {defaultValue=true, parameterName=Use Noise}
   Mobile=@n138
  ) {desc=Migrated from FeatureLevelSwitch}
  Reroute(None=Multiply(A=LinearInterpolate(B=@n126, Alpha=@n127) {constA=1}, B=QualitySwitch(Default=@n139, Low=@n138, High=@n138, Medium=@n138))) {desc=Opacity}
  Reroute(None=@n62)
  Reroute(None=@n98)
  Reroute()
  ScalarParameter() {defaultValue=0.267, parameterName=ConeTangent}
  ScalarParameter() {defaultValue=186, parameterName=ZMin}
  ScalarParameter() {defaultValue=1186, parameterName=ZMax}
  ScalarParameter() {parameterName=RadiusEnd}
  ScalarParameter() {defaultValue=0.0001, parameterName=testes}
  ShadingPathSwitch(Default=@n78, Mobile=@n76) {desc=Migrated from FeatureLevelSwitch}
  n140 = ObjectPositionWS()
  n141 = WorldPosition()
  n142 = WorldPosition()
  n143 = NamedRerouteUsage()
  n144 = NamedRerouteUsage()
  n145 = LinearInterpolate(A=Multiply(A=NamedRerouteUsage(), B=Subtract(A=@n143, B=@n144)), Alpha=VertexColor().R) {constB=0}
  n146 = ScalarParameter() {defaultValue=8, group=Advanced, parameterName=Overscan for Near Clip}
  n147 = TextureCoordinate() {coordinateIndex=1}
  n148 = MaterialFunctionCall(Float2=Multiply(A=Add(A=@n145, B=@n146), B=ConstantBiasScale(None=@n147) {bias=-0.5, desc=-1 tp 1 Circular UVs, scale=2})) {materialFunction=&BreakOutFloat2Components}
  n149 = NamedRerouteUsage()
  n150 = ShadingPathSwitch(Default=StaticSwitchParameter(True=NamedRerouteUsage(), False=@n149) {parameterName=UseWpoAnimation}, Mobile=@n149) {desc=Migrated from FeatureLevelSwitch}
  n151 = NamedRerouteUsage()
  n152 = ShadingPathSwitch(Default=StaticSwitchParameter(True=NamedRerouteUsage(), False=@n151) {parameterName=UseWpoAnimation}, Mobile=@n151) {desc=Migrated from FeatureLevelSwitch}
  n153 = Add(
   A=Multiply(A=@n148.R, B=QualitySwitch(Default=@n150, Low=@n149, High=@n149, Medium=@n149))
   B=Multiply(A=@n148.G, B=QualitySwitch(Default=@n152, Low=@n151, High=@n151, Medium=@n151))
  )
  n154 = Add(
   A=@n153
   B=RotateAboutAxis(
    NormalizedRotationAxis=NamedRerouteUsage()
    RotationAngle=NamedRerouteUsage()
    PivotPoint=NamedRerouteUsage()
    Position=Add(A=@n142.XYZ, B=@n153)
   )
  )
  n155 = QualitySwitch(
   Default=ShadingPathSwitch(Default=StaticSwitchParameter(True=@n154, False=@n153) {parameterName=UseWpoAnimation}, Mobile=@n153) {desc=Migrated from FeatureLevelSwitch}
   Low=@n153
   High=@n153
   Medium=@n153
  )
  n156 = StaticSwitchParameter(True=LinearInterpolate(A=Subtract(A=@n140, B=@n141.XYZ), B=@n155, Alpha=NamedRerouteUsage()), False=@n155) {parameterName=UseCustomPrimInstanceData}
  n157 = NamedRerouteUsage()
  n158 = Add(A=@n156, B=WorldPosition().XYZ)
  n159 = ScalarParameter() {defaultValue=0.376, parameterName=LowDetailDownScale}
  n160 = Add(A=@n156, B=Multiply(A=Subtract(A=@n157, B=@n158), B=OneMinus(None=@n159)))
  StaticSwitchParameter(True=QualitySwitch(Default=@n156, Low=@n160, High=@n160, Medium=@n160), False=@n156) {defaultValue=true, parameterName=UseLowDetailDownScale}
  TransformPosition(None=Constant3Vector()) {transformType=TRANSFORMPOSSOURCE_World}
 ]
}
